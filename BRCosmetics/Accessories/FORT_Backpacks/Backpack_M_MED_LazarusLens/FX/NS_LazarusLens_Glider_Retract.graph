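# One-shot Niagara burst for the LazarusLens glider retract: a red flare and smoke puff, offset by a shared system MasterOffset.
# DarkGradient backs two flares, one casting light and one a wide streak, with camera offset driven by view angle to vertical.
# Smoke uses curl noise, drag, SubUV animation and fading inherited source movement, plus brief SmokeDetail sprites.
# Front-end modifiers lower smoke velocity outside gameplay; dynamic material parameters feed every layer.
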NiagaraSystem /BRCosmetics/Accessories/FORT_Backpacks/Backpack_M_MED_LazarusLens/FX/NS_LazarusLens_Glider_Retract
dependencies {
 usedDataInterfaces [
  &NiagaraDataInterfaceCurve
  &NiagaraDataInterfaceCamera
  &NiagaraDataInterfaceVectorCurve
  &NiagaraDataInterfaceSpriteRendererInfo
 ]
 usedDynamicInputs [&ConvertVectorToPosition]
 usedModules [
  &self:SystemSpawnScript.SystemScriptSource.SystemScriptGraph.NiagaraNodeAssignment_0.SetVariables_#1
  &SystemState
 ]
}
emitters {
 DarkGradient {
  emitterUpdateScript [
   &EmitterState()
   &SpawnBurst_Instantaneous(Use Spawn Probability=false, UseLoopCountLimit=false)
  ]
  particleSpawnScript [
   &InitializeParticle(
    Lifetime=0.4
    Position Offset=$System.MasterOffset
    Sprite Size=<325 325>
    Sprite Size Mode=::NewEnumerator4
    UsePositionOffset=true
    Write Material Random=true
   )
   &Color(Color=&MakeLinearColorFromVectorAndFloat("Vector (RGB)"=<0.003 0.0098 0.0156>, "Float (Alpha)"=&ClampFloat(Float=1.2)))
  ]
  particleUpdateScript [
   &ParticleState()
   &ScaleColor(
    Scale Alpha=&FloatFromCurve(FloatCurve=DataInterface() {propertyValues {Curve {keys [<at 0 0 RCIM_Cubic>,<at 0.4043 1 RCIM_Cubic>,<at 1 0>]}}})
    ScaleA=true
    ScaleRGB=true
   )
   &DynamicMaterialParameters(
    Index 0 Param 1=1
    Index 0 Param 2=0.1
    Index 0 Param 3=1
    Index 0 Param 4=1
    Module.Param0WriteEnabled=true
    Module.Param1WriteEnabled=true
    Module.Param2WriteEnabled=true
    Module.Param3WriteEnabled=true
   )
   &CameraOffset(Camera Offset Amount=50)
  ]
  rendererClasses [&NiagaraSpriteRendererProperties]
  renderers [{rendererClass=&NiagaraSpriteRendererProperties}]
 }
 Flare {
  emitterUpdateScript [
   &EmitterState()
   &SpawnBurst_Instantaneous(Use Spawn Probability=false, UseLoopCountLimit=false)
  ]
  particleSpawnScript [&InitializeParticle(
    Color=<rgba 1.25 0 0.0868 1>
    Lifetime=0.4
    Position Offset=$System.MasterOffset
    Sprite Size=&Multiply_Vector2D_ByFloat(Float=0.55, Vector2D=<2500 2000>)
    Sprite Size Mode=::NewEnumerator4
    UsePositionOffset=true
    Write Material Random=true
   )]
  particleUpdateScript [
   &ParticleState()
   &ScaleSpriteSize()
   &CameraOffset(Camera Offset Amount=&self:Dot(
     A:Vector3f=&ReturnCameraVector(Camera Vectors=::NewEnumerator1, Coordinate Space=::World)
     B:Vector3f=<0 0 1>
     Curve for Floats:NiagaraDataInterfaceCurve=DataInterface() {propertyValues {Curve {keys [<at 0 100>,<at 0.5 75>,<at 1 25>]}}}
    ))
   &ScaleColor(
    Scale Alpha=&FloatFromCurve(FloatCurve=DataInterface() {propertyValues {Curve {keys [
         <at 0 0 RCIM_Cubic>
         {arriveTangent=0.3473, interpMode=RCIM_Cubic, leaveTangent=0.3473, time=0.1032, value=0.3982}
         {arriveTangent=-0.4517, interpMode=RCIM_Cubic, leaveTangent=-0.4517, tangentMode=RCTM_User, time=0.2365, value=0.0822}
         <at 0.75 0>
        ]}}})
    Scale RGB=&VectorFromCurve(
     VectorCurve=DataInterface() {propertyValues {
       XCurve {keys [<at 0 0.9384 RCIM_Cubic>,<at 0.5 1 RCIM_Cubic>]}
       YCurve {keys [<at 0 0.3016 RCIM_Cubic>,<at 0.5 0.05 RCIM_Cubic>]}
       ZCurve {keys [<at 0 0.5227 RCIM_Cubic>,<at 0.5 0 RCIM_Cubic>]}
      }}
     Scale Curve=&VectorFromFloat(Value=3)
    )
    ScaleA=true
    ScaleRGB=true
   )
   &DynamicMaterialParameters(
    Index 0 Param 1=1.5
    Index 0 Param 2=1
    Index 0 Param 3=1
    Index 0 Param 4=1
    Module.Param0WriteEnabled=true
    Module.Param1WriteEnabled=true
    Module.Param2WriteEnabled=true
    Module.Param3WriteEnabled=true
   )
  ]
  rendererClasses [&NiagaraSpriteRendererProperties, &NiagaraLightRendererProperties]
  renderers [
   {rendererClass=&NiagaraSpriteRendererProperties}
   {rendererClass=&NiagaraLightRendererProperties, rendererIndex=1}
  ]
 }
 Flare001 {
  emitterUpdateScript [
   &EmitterState()
   &SpawnBurst_Instantaneous(Use Spawn Probability=false, UseLoopCountLimit=false)
  ]
  particleSpawnScript [&InitializeParticle(
    Color=<rgba 1.5 1.5 1.5 1>
    Lifetime=0.42
    Position Offset=$System.MasterOffset
    Sprite Size=&Multiply_Vector2D_ByFloat(Float=0.55, Vector2D=<3000 500>)
    Sprite Size Mode=::NewEnumerator4
    UsePositionOffset=true
    Write Material Random=true
   )]
  particleUpdateScript [
   &ParticleState()
   &ScaleSpriteSize()
   &CameraOffset(Camera Offset Amount=&self:Dot(
     A:Vector3f=&ReturnCameraVector(Camera Vectors=::NewEnumerator1, Coordinate Space=::World)
     B:Vector3f=<0 0 1>
     Curve for Floats:NiagaraDataInterfaceCurve=DataInterface() {propertyValues {Curve {keys [<at 0 100>,<at 0.5 75>,<at 1 25>]}}}
    ))
   &ScaleColor(
    Scale Alpha=&FloatFromCurve(FloatCurve=DataInterface() {propertyValues {Curve {keys [
         <at 0 0 RCIM_Cubic>
         {arriveTangent=0.3473, interpMode=RCIM_Cubic, leaveTangent=0.3473, time=0.1032, value=0.3982}
         {arriveTangent=-0.444, interpMode=RCIM_Cubic, leaveTangent=-0.444, time=0.2365, value=0.0822}
         <at 1 0>
        ]}}})
    Scale RGB=&VectorFromCurve(
     VectorCurve=DataInterface() {propertyValues {
       XCurve {keys [<at 0 0.9384 RCIM_Cubic>,<at 0.5 1 RCIM_Cubic>]}
       YCurve {keys [<at 0 0.3016 RCIM_Cubic>,<at 0.5 0.05 RCIM_Cubic>]}
       ZCurve {keys [<at 0 0.5227 RCIM_Cubic>,<at 0.5 0 RCIM_Cubic>]}
      }}
     Scale Curve=&VectorFromFloat(Value=2)
    )
    ScaleA=true
    ScaleRGB=true
   )
   &DynamicMaterialParameters(
    Index 0 Param 1=1.5
    Index 0 Param 2=1
    Index 0 Param 3=1
    Index 0 Param 4=1
    Module.Param0WriteEnabled=true
    Module.Param1WriteEnabled=true
    Module.Param2WriteEnabled=true
    Module.Param3WriteEnabled=true
   )
  ]
  rendererClasses [&NiagaraSpriteRendererProperties]
  renderers [{rendererClass=&NiagaraSpriteRendererProperties}]
 }
 Smoke {
  emitterUpdateScript [
   &EmitterState()
   &SpawnBurst_Instantaneous(Spawn Count=25, Use Spawn Probability=false, UseLoopCountLimit=false)
  ]
  particleSpawnScript [
   &InitializeParticle(
    Lifetime=&RandomRangeFloat(Minimum=0.8, Maximum=1.2)
    Position Offset=$System.MasterOffset
    Position Offset Coordinate Space=::Local
    Sprite Rotation Angle=&RandomRangeFloat(Minimum=0, Maximum=1)
    Sprite Rotation Mode=::NewEnumerator1
    Sprite Size=&Multiply_Vector2D_ByFloat(Float=0.5, Vector2D=&RandomRangeVector2D(Minimum=<400 500>, Maximum=<600 900>))
    Sprite Size Mode=::NewEnumerator4
    UsePositionOffset=true
    Write Material Random=true
   )
   &AddVelocity(Velocity=&Multiply_VectorByFloat(
     Vector=<1 1 1>
     Float=&FN_FrontEndFloatModifier(Gameplay=&RandomRangeFloat(Minimum=5, Maximum=15), Front End=5)
    )) {module=AddVelocity001}
   &AddVelocity(
    Velocity=&FN_FrontEndVector3Modifier(
     Gameplay=&RandomRangeVector(Minimum=<-500 -500 -500>, Maximum=<500 500 500>)
     Front End=&RandomRangeVector(Minimum=<-250 -250 -250>, Maximum=<250 250 250>)
    )
    Velocity Speed Scale=0.5
   )
   &Color(Color=&MakeLinearColorFromVectorAndFloat(
     "Vector (RGB)"=&VectorFromFloat(Value=&RandomRangeFloat(Minimum=0, Maximum=1))
     "Float (Alpha)"=&ClampFloat(Float=&RandomRangeFloat(Minimum=0.6, Maximum=0.8))
    ))
   &ShapeLocation(MaskSpawnGroup=false, Offset Mode=::NewEnumerator0, Override Seed=false, Sphere Radius=20, UseOwnerScale=true) {enabled=false}
   &SpriteFacingAndAlignment(WriteAlignment=false, WriteFacing=true)
   /SetVariables(Particles.Linear Color=&RandomRangeLinearColor(Minimum=<rgba 0 0 0 0.08>, Maximum=<rgba 0.15 0.0451 0.0336 0.125>)) {isSetParameters=true}
   /SetVariables(Particles.VisibilityTag=&RandomRangeInt(Minimum=1, Maximum=1)) {isSetParameters=true}
  ]
  particleUpdateScript [
   &ParticleState()
   &SpriteRotationRate(Rotation Rate=&Multiply_Float(A=60, B=&RandomRangeFloat(Minimum=-1, Maximum=1)))
   &ScaleSpriteSize(
    Uniform Curve Scale=&RandomRangeFloat(Minimum=0.75, Maximum=1.25)
    Uniform Curve Sprite Scale=DataInterface() {propertyValues {Curve {keys [
        {arriveTangent=3.4179, interpMode=RCIM_Cubic, leaveTangent=3.4179, tangentMode=RCTM_User, time=0, value=0}
        {arriveTangent=2.6393, interpMode=RCIM_Cubic, leaveTangent=2.6393, tangentMode=RCTM_User, time=1, value=1}
       ]}}}
   )
   &Drag(Drag=3, Use Linear Drag=true, Use Rotational Drag=true)
   &InheritSourceMovement(
    Applied Position Scale=&VectorFromFloat(Value=&FloatFromCurve(FloatCurve=DataInterface() {propertyValues {Curve {keys [
          {
           arriveTangentWeight=0.8189
           interpMode=RCIM_Cubic
           leaveTangentWeight=0.8189
           tangentMode=RCTM_User
           tangentWeightMode=RCTWM_WeightedBoth
           time=0
           value=1
          }
          {arriveTangent=-3.0942, interpMode=RCIM_Cubic, leaveTangent=-3.0942, tangentMode=RCTM_User, time=1, value=0}
         ]}}}))
    ApplyForce=false
    ApplyPosition=true
    ApplyPositionFalloff=false
   )
   &CurlNoiseForce(Noise Strength=1500, Pan Noise=false)
   &SolveForcesAndVelocity(Clamp Velocity=false, Limit Acceleration=false)
   /Niagara/Modules/Update/SubUV/SubUVAnimation(Number Of Frames=0)
   /SetVariables(Particles.Sub Image 2:NiagaraFloat=$Particles.SubImageIndex) {isSetParameters=true}
   /Niagara/Modules/Update/SubUV/V2/SubUVAnimation(
    Random Start Frame=true
    "Sprite Renderer "=DataInterface() {propertyValues {SpriteRenderer=&self:Smoke_6.NiagaraSpriteRendererProperties_1}}
    SubUV Lookup Index Scale=0.65
    UseEndFrame=false
    UseStartFrame=false
   ) {module=SubUVAnimation001}
   &DynamicMaterialParameters(
    Index 0 Param 1=&FloatFromCurve(FloatCurve=DataInterface() {propertyValues {Curve {keys [<at -0.0006 1.5633 RCIM_Cubic>,<at 1 0 RCIM_Cubic>]}}})
    Index 0 Param 2=&FloatFromCurve(FloatCurve=DataInterface() {propertyValues {Curve {keys [
         <at 0.0026 0.2349 RCIM_Cubic>
         {arriveTangent=-0.1916, interpMode=RCIM_Cubic, leaveTangent=-0.1916, time=0.244, value=0.7628}
         {arriveTangent=0.4577, interpMode=RCIM_Cubic, leaveTangent=0.4577, time=0.8984, value=0.0632}
         <at 0.9992 1.1085 RCIM_Cubic>
        ]}}})
    Index 0 Param 3=1
    Index 0 Param 4=1
    Module.Param0WriteEnabled=true
    Module.Param1WriteEnabled=true
    Module.Param2WriteEnabled=true
    Module.Param3WriteEnabled=true
   )
   &ScaleColor(
    Scale Alpha=&FloatFromCurve(FloatCurve=DataInterface() {propertyValues {Curve {keys [<at 0 0 RCIM_Cubic>,<at 0.0662 2.5783 RCIM_Cubic>,<at 1 0 RCIM_Cubic>]}}})
    ScaleA=true
    ScaleRGB=true
   )
   &CameraOffset(Camera Offset Amount=&RandomRangeFloat(Minimum=-20, Maximum=35))
  ]
  rendererClasses [&NiagaraSpriteRendererProperties]
  renderers [
   {rendererClass=&NiagaraSpriteRendererProperties}
   {rendererClass=&NiagaraSpriteRendererProperties, rendererIndex=1}
  ]
 }
 SmokeDetail {
  emitterUpdateScript [
   &EmitterState()
   &SpawnBurst_Instantaneous(Spawn Count=2, Use Spawn Probability=false, UseLoopCountLimit=false)
  ]
  particleSpawnScript [
   &InitializeParticle(
    Lifetime=0.3
    Position Offset=$System.MasterOffset
    Sprite Rotation Mode=::NewEnumerator2
    Sprite Size=&RandomRangeVector2D(Minimum=<300 300>, Maximum=<600 600>)
    Sprite Size Mode=::NewEnumerator4
    UsePositionOffset=true
    Write Material Random=true
   )
   &Color(Color=&MakeLinearColorFromVectorAndFloat("Vector (RGB)"=<0.0677 0.0282 0.0282>, "Float (Alpha)"=&ClampFloat(Float=0.75)))
   &DynamicMaterialParameters(
    Index 0 Param 2=&RandomRangeFloat(Minimum=0, Maximum=5)
    Module.Param0WriteEnabled=false
    Module.Param1WriteEnabled=true
    Module.Param2WriteEnabled=false
    Module.Param3WriteEnabled=false
   ) {module=DynamicMaterialParameters001}
   &CameraOffset(Camera Offset Amount=50)
  ]
  particleUpdateScript [
   &ParticleState()
   &ScaleSpriteSize()
   &ScaleColor(
    Scale Alpha=&FloatFromCurve(FloatCurve=DataInterface() {propertyValues {Curve {keys [
         {arriveTangent=2.6336, interpMode=RCIM_Cubic, time=0.003, value=-0.0044}
         {arriveTangent=0.4503, interpMode=RCIM_Cubic, leaveTangent=0.4503, time=0.1124, value=1.5409}
         {arriveTangent=-1.7361, interpMode=RCIM_Cubic, leaveTangent=-1.7361, time=0.5926, value=0.2611}
         <at 1 0>
        ]}}})
    Scale RGB=&VectorFromFloat(Value=&FloatFromCurve(FloatCurve=DataInterface() {propertyValues {Curve {keys [<at 0 1 RCIM_Cubic>,<at 1 0>]}}}))
    ScaleA=true
    ScaleRGB=true
   )
   &DynamicMaterialParameters(
    Index 0 Param 1=&FloatFromCurve(FloatCurve=DataInterface() {propertyValues {Curve {keys [{arriveTangent=3.783, interpMode=RCIM_Cubic, time=0, value=0}, <at 1 1.3296 RCIM_Cubic>]}}})
    Index 0 Param 3=1
    Index 0 Param 4=1
    Module.Param0WriteEnabled=true
    Module.Param1WriteEnabled=false
    Module.Param2WriteEnabled=true
    Module.Param3WriteEnabled=true
   )
   &InheritSourceMovement(
    Applied Position Scale=&VectorFromFloat(Value=&FloatFromCurve(FloatCurve=DataInterface() {propertyValues {Curve {keys [<at 0 1 RCIM_Cubic>, {arriveTangent=-2.9092, interpMode=RCIM_Cubic, leaveTangent=-2.9092, tangentMode=RCTM_User, time=1, value=0}]}}}))
    ApplyForce=false
    ApplyPosition=true
    ApplyPositionFalloff=false
   ) {enabled=false}
   &SolveForcesAndVelocity(Clamp Velocity=false, Limit Acceleration=false) {enabled=false}
  ]
  rendererClasses [&NiagaraSpriteRendererProperties]
  renderers [{rendererClass=&NiagaraSpriteRendererProperties}]
 }
}
summary {
 systemName=NS_LazarusLens_Glider_Retract
 emitters [
  {emitterName=DarkGradient, rendererClasses [&NiagaraSpriteRendererProperties]}
  {emitterName=Flare, rendererClasses [&NiagaraSpriteRendererProperties, &NiagaraLightRendererProperties]}
  {emitterName=Flare001, rendererClasses [&NiagaraSpriteRendererProperties]}
  {emitterName=SmokeDetail, rendererClasses [&NiagaraSpriteRendererProperties]}
  {emitterName=Smoke, rendererClasses [&NiagaraSpriteRendererProperties]}
 ]
 userVariables [
  User.bIsFrontEnd {type=NiagaraBool}
  User.bIsFrontEndPreview {type=NiagaraBool}
 ]
}
